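# Physics asset for the Slurp Jellyfish fish gadget mesh, giving its body and tentacles capsule collision and joint limits.
# The Anim capsule anchors a three-link dyn_spineFront chain and six two-segment tentacles (A, B, C, E, F, G; no D).
# Tentacle joints share limited swing and tight twist, and the upper spine links are tighter in swing.
# The spine's root joint to Anim is free on all angular axes.
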PhysicsAsset /Game/Gadgets/Assets/Fish/Slurp_Jellyfish/Meshes/Slurp_Jellyfish_Physics
bodies {
 Anim [{shapeType=Capsule, length=11.6521, radius=9.2557, center=<0.1298 0.1511 6.9471>, rotation=<rot -1.5731 -90 0.1237>}]
 dyn_spineFront_01 [{shapeType=Capsule, length=0.505, radius=0.505, rotation=<rot 0 -90 0>}]
 dyn_spineFront_02 [{shapeType=Capsule, length=0.1193, radius=6.816, center=<3.2403 0 0.0001>, rotation=<rot 0 -90 0>}]
 dyn_spineFront_03 [{shapeType=Capsule, length=13.8715, radius=12.2306, center=<2.5811 0.171 -0.0357>, rotation=<rot 87.8113 75.0111 165.024>}]
 tentacle_A_01 [{shapeType=Capsule, length=8.3949, radius=1.4853, center=<5.5 0 0>, rotation=<rot 0 89.0397 -89.5497>}]
 tentacle_A_02 [{shapeType=Capsule, length=14.9809, radius=1.6669, center=<5.5 0 0>, rotation=<rot 0 89.9954 -89.6599>}]
 tentacle_B_01 [{shapeType=Capsule, length=8.0824, radius=1.4041, center=<5.5 0 0>, rotation=<rot 0 90.579 -89.6132>}]
 tentacle_B_02 [{shapeType=Capsule, length=11.6502, radius=1.8449, center=<5.5 0 0>, rotation=<rot 0 89.3488 -89.6808>}]
 tentacle_C_01 [{shapeType=Capsule, length=7.7487, radius=1.4558, center=<5.5 0 0>, rotation=<rot 0 89.6351 -89.5269>}]
 tentacle_C_02 [{shapeType=Capsule, length=11.0904, radius=1.6933, center=<5.5 0 0>, rotation=<rot 0 89.8144 -89.9275>}]
 tentacle_E_01 [{shapeType=Capsule, length=11.9525, radius=1.4067, center=<5.5 0 0>, rotation=<rot 0 -91.742 -89.2481>}]
 tentacle_E_02 [{shapeType=Capsule, length=12.9884, radius=1.7027, center=<5.5 0 0>, rotation=<rot 0 -90.6916 -89.9542>}]
 tentacle_F_01 [{shapeType=Capsule, length=8.1453, radius=1.3891, center=<5.5 0 0>, rotation=<rot 0 90.6025 -89.6131>}]
 tentacle_F_02 [{shapeType=Capsule, length=11.6368, radius=1.6408, center=<5.5 0 0>, rotation=<rot 0 89.2756 -89.8894>}]
 tentacle_G_01 [{shapeType=Capsule, length=8.045, radius=1.3941, center=<5.5 0 0>, rotation=<rot 0 90.2143 -89.4485>}]
 tentacle_G_02 [{shapeType=Capsule, length=11.275, radius=1.6166, center=<5.5 0 0>, rotation=<rot 0 -90.546 -89.3599>}]
}
constraints [
 {bone1Name=tentacle_A_01, bone2Name=Anim, swing1LimitDegrees=30, swing2LimitDegrees=40, twistLimitDegrees=8}
 {bone1Name=tentacle_A_02, bone2Name=tentacle_A_01, swing1LimitDegrees=30, swing2LimitDegrees=40, twistLimitDegrees=8}
 {bone1Name=tentacle_B_01, bone2Name=Anim, swing1LimitDegrees=30, swing2LimitDegrees=40, twistLimitDegrees=8}
 {bone1Name=tentacle_B_02, bone2Name=tentacle_B_01, swing1LimitDegrees=30, swing2LimitDegrees=40, twistLimitDegrees=8}
 {bone1Name=tentacle_C_01, bone2Name=Anim, swing1LimitDegrees=30, swing2LimitDegrees=40, twistLimitDegrees=8}
 {bone1Name=tentacle_C_02, bone2Name=tentacle_C_01, swing1LimitDegrees=30, swing2LimitDegrees=40, twistLimitDegrees=8}
 {bone1Name=tentacle_E_01, bone2Name=Anim, swing1LimitDegrees=30, swing2LimitDegrees=40, twistLimitDegrees=8}
 {bone1Name=tentacle_E_02, bone2Name=tentacle_E_01, swing1LimitDegrees=30, swing2LimitDegrees=40, twistLimitDegrees=8}
 {bone1Name=tentacle_F_01, bone2Name=Anim, swing1LimitDegrees=30, swing2LimitDegrees=40, twistLimitDegrees=8}
 {bone1Name=tentacle_F_02, bone2Name=tentacle_F_01, swing1LimitDegrees=30, swing2LimitDegrees=40, twistLimitDegrees=8}
 {bone1Name=tentacle_G_01, bone2Name=Anim, swing1LimitDegrees=30, swing2LimitDegrees=40, twistLimitDegrees=8}
 {bone1Name=tentacle_G_02, bone2Name=tentacle_G_01, swing1LimitDegrees=30, swing2LimitDegrees=40, twistLimitDegrees=8}
 {bone1Name=dyn_spineFront_02, bone2Name=dyn_spineFront_01, swing1LimitDegrees=20, swing2LimitDegrees=20, twistLimitDegrees=8}
 {
  bone1Name=dyn_spineFront_01
  bone2Name=Anim
  swing1LimitDegrees=45
  swing1Motion=Free
  swing2LimitDegrees=45
  swing2Motion=Free
  twistLimitDegrees=45
  twistMotion=Free
 }
 {bone1Name=dyn_spineFront_03, bone2Name=dyn_spineFront_02, swing1LimitDegrees=20, swing2LimitDegrees=20, twistLimitDegrees=8}
]
