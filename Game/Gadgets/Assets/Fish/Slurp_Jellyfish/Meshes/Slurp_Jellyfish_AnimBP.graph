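# Animation Blueprint for the Slurp Jellyfish gadget asset, giving it purely physics-driven secondary motion with no animation playback.
# The local-space ref pose feeds a RigidBody node with an override physics asset, reduced downward gravity and scaled, clamped component velocity and acceleration, then goes to the output.
# Trail controllers for the front spine and tentacles remain unconnected to the output.
# The event graph holds only unwired default nodes.
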AnimBlueprint /Game/Gadgets/Assets/Fish/Slurp_Jellyfish/Meshes/Slurp_Jellyfish_AnimBP
target_skeleton=&/Slurp_Jellyfish_Skeleton
graphs [
 AnimGraph {nodes [
   Component To Local()
   Component To Local()
   n1 = RigidBody(ComponentPose=Local To Component(LocalPose=Local Space Ref Pose())) {
    bClampLinearTranslationLimitToRefPose=true
    bOverrideWorldGravity=true
    overridePhysicsAsset=&/Slurp_Jellyfish_Physics
    alphaBoolBlend {blendOption=Linear}
    componentAppliedLinearAccClamp=<1500 1500 1500>
    componentLinearAccScale=<0.2 0.2 0.2>
    componentLinearVelScale=<0.1 0.1 0.1>
    overrideWorldGravity {z=-100}
   }
   Output Pose(Result=Component To Local(ComponentPose=@n1)) {layerGroup=None}
   Trail controller - Bone: dyn_spineFront_03!(ComponentPose, Alpha=0.3, RelaxationSpeedScale:float=4) -> (Pose) {
    chainLength=3
    relaxationSpeedScale=4
    rotationLimits [{limitMax=<180 180 180>}, {limitMax=<180 180 180>}, {limitMax=<180 180 180>}]
    rotationOffsets [<0 0 0>, <0 0 0>, <0 0 0>]
    trailBone {boneName=dyn_spineFront_03}
    bLimitStretch=~
   }
   n2 = Trail controller - Bone: tentacle_A_02!(ComponentPose, Alpha=0.2, RelaxationSpeedScale:float=1) -> (Pose) {
    relaxationSpeedScale=4
    rotationLimits [{limitMax=<180 180 180>}, {limitMax=<180 180 180>}]
    rotationOffsets [<0 0 0>, <0 0 0>]
    trailBone {boneName=tentacle_A_02}
   }
   n3 = Trail controller - Bone: tentacle_B_02!(ComponentPose=@n2, Alpha=0.3, RelaxationSpeedScale:float=1) -> (Pose) {
    relaxationSpeedScale=4
    rotationLimits [{limitMax=<180 180 180>}, {limitMax=<180 180 180>}]
    rotationOffsets [<0 0 0>, <0 0 0>]
    trailBone {boneName=tentacle_B_02}
   }
   n4 = Trail controller - Bone: tentacle_C_02!(ComponentPose=@n3, Alpha=0.25, RelaxationSpeedScale:float=1) -> (Pose) {
    relaxationSpeedScale=4
    rotationLimits [{limitMax=<180 180 180>}, {limitMax=<180 180 180>}]
    rotationOffsets [<0 0 0>, <0 0 0>]
    trailBone {boneName=tentacle_C_02}
   }
   n5 = Trail controller - Bone: tentacle_E_02!(ComponentPose=@n4, Alpha=0.35, RelaxationSpeedScale:float=1) -> (Pose) {
    relaxationSpeedScale=4
    rotationLimits [{limitMax=<180 180 180>}, {limitMax=<180 180 180>}]
    rotationOffsets [<0 0 0>, <0 0 0>]
    trailBone {boneName=tentacle_E_02}
   }
   n6 = Trail controller - Bone: tentacle_F_02!(ComponentPose=@n5, Alpha=0.22, RelaxationSpeedScale:float=1) -> (Pose) {
    relaxationSpeedScale=4
    rotationLimits [{limitMax=<180 180 180>}, {limitMax=<180 180 180>}]
    rotationOffsets [<0 0 0>, <0 0 0>]
    trailBone {boneName=tentacle_F_02}
   }
   Trail controller - Bone: tentacle_G_02!(ComponentPose=@n6, Alpha=0.235, RelaxationSpeedScale:float=1) -> (Pose) {
    relaxationSpeedScale=4
    rotationLimits [{limitMax=<180 180 180>}, {limitMax=<180 180 180>}]
    rotationOffsets [<0 0 0>, <0 0 0>]
    trailBone {boneName=tentacle_G_02}
   }
  ]}
 EventGraph {nodes [
   TryGetPawnOwner()
   Event BlueprintUpdateAnimation()
  ]}
]
variables []
